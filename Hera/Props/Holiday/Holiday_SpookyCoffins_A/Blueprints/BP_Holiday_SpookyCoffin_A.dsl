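; Holiday spooky coffin prop Blueprint that colors its separate coffin top mesh from the building's texture data.
; The construction script calls the parent, looks up the texture data entry at a local index and passes its tint color to the secondary-mesh color function for the top.
; An invalid index falls back to mid gray.
; BeginPlay, overlap and Tick events are empty.
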(event EventBeginPlay)

(event Collision|EventActorBeginOverlap
  (params (OtherActor :type Actor%20Object%20Reference)))

(event EventTick
  (params (DeltaSeconds :type Float%20%28single-precision%29)))

(fn ConstructionScript
  (bind _sm_holiday_spookycoffintop_a (Variables|Default|GetSM_Holiday_SpookyCoffinTop_A))
  (bind _l_tdindex (LocalVariables|GetL_TDIndex))
  (|CallParentFunction)
  (bind _outtexturedataarray (Building|GetBuildingTextureData))
  (if (Utilities|Array|IsValidIndex _outtexturedataarray _l_tdindex)
    (Default|SetColoronSecondaryMesh (Math|Conversions|ToLinearColor(Color) (Variables|/Script/FortniteGame.BuildingTextureData|GetTintColor (Utilities|Array|Get(acopy) _outtexturedataarray _l_tdindex))) _sm_holiday_spookycoffintop_a)
    (else
      (Default|SetColoronSecondaryMesh "(R=0.500000,G=0.500000,B=0.500000,A=1.000000)" _sm_holiday_spookycoffintop_a))))
